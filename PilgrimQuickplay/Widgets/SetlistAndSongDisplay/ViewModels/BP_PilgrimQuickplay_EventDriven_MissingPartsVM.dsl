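; Event-driven Pilgrim Quickplay view model that flags when a song lacks a part for the player's chosen instrument.
; Once the event router is available, it caches upcoming and current songs, then refreshes only the upcoming one on instrument or custom setlist change events.
; bIsMissingPart compares the upcoming song's track intensity to the missing-track value.
; CheckIfHidePlayerStats applies the same test to the cached played song.
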(event Sparks|ViewModel|EventBP_OnGlobalEventRouterAvailable
  (params (EventRouter :type Gameplay%20Event%20Router%20Component%20Object%20Reference) (PlayerState :type Fort%20Player%20State%20Object%20Reference) (OwningLocalPlayer :type Local%20Player%20Object%20Reference) (OwningPlayerController :type Fort%20Player%20Controller%20Object%20Reference))
  (bind _eventrouter (Pilgrim|PilgrimWaitForEventRouter OwningPlayerController)
    (:OnEventRouterAvailable
      (Variables|Default|SetEvent%20Router%20Ref _eventrouter)
      (Utilities|FlowControl|Sequence
        (:then_0
          (bind _asynctaskproxy (GameplayEventRouter|StartListeningToEvent (Sparks|ViewModel|BPGetOwningPlayerController) (Variables|Default|GetEvent%20Router%20Ref) "/Script/PilgrimCoreRuntime.PilgrimEvent_PlayerState_InstrumentChanged" (Variables|/Script/Engine.Controller|GetPlayerState (Sparks|ViewModel|BPGetOwningPlayerController)))
            (:then
              (Sparks|ViewModel|TrackEventHandle _asynctaskproxy)
              (bind _asynctaskproxy_2 (GameplayEventRouter|StartListeningToEvent (Sparks|ViewModel|BPGetOwningPlayerController) (Variables|Default|GetEvent%20Router%20Ref) "/Script/PilgrimCoreRuntime.PilgrimEvent_BandState_CustomSetlistChanged")
                (:then
                  (Sparks|ViewModel|TrackEventHandle _asynctaskproxy_2))
                (:OnEventReceived
                  (CallFunction|SettingSongMetadata))))
            (:OnEventReceived
              (CallFunction|SettingSongMetadata))))
        (:then_1
          (CallFunction|SettingSongMetadata)
          (CallFunction|SettingCompletedSongMetadata))))))

(fn CheckIfHidePlayerStats
  (Utilities|IsValid (Variables|Components|GetPlayedSongMetadata)
    (:Is%20Valid
      (bind _returnvalue (Class|PilgrimCorePlayerStateComponent|GetPilgrimCorePlayerStateComponent (Variables|/Script/Engine.Controller|GetPlayerState (Sparks|ViewModel|BPGetOwningPlayerController))))
      (Utilities|IsValid _returnvalue
        (:Is%20Valid
          (return (== (Meta|GetMissingTrackIntensity) (Class|PilgrimSongMetadata|GetTrackIntensity (Variables|Components|GetPlayedSongMetadata) (Class|PilgrimCorePlayerStateComponent|GetInstrumentChoice _returnvalue)))))))))

(fn DetermineIfMissingPart
  (params (SongMetadata :type Pilgrim%20Song%20Metadata%20Object%20Reference))
  (Utilities|IsValid SongMetadata
    (:Is%20Valid
      (bind _returnvalue (Class|PilgrimCorePlayerStateComponent|GetPilgrimCorePlayerStateComponent (Variables|/Script/Engine.Controller|GetPlayerState (Sparks|ViewModel|BPGetOwningPlayerController))))
      (Utilities|IsValid _returnvalue
        (:Is%20Valid
          (if (== (Meta|GetMissingTrackIntensity) (Class|PilgrimSongMetadata|GetTrackIntensity SongMetadata (Class|PilgrimCorePlayerStateComponent|GetInstrumentChoice _returnvalue)))
            (Variables|Default|SetbIsMissingPart true)
            (else
              (Variables|Default|SetbIsMissingPart false))))))))

(fn SettingCompletedSongMetadata
  (bind _returnvalue (Sparks|ViewModel|BPGetOwningPlayerController))
  (bind _returnvalue_2 (Class|PilgrimQuickplayBPFL|GetCurrentSong _returnvalue))
  (Utilities|IsValid _returnvalue
    (:Is%20Valid
      (if (Utilities|IsValid _returnvalue_2 :!class /Script/Engine.KismetSystemLibrary)
        (Variables|Components|SetPlayedSongMetadata _returnvalue_2)))))

(fn SettingSongMetadata
  (bind _returnvalue (Sparks|ViewModel|BPGetOwningPlayerController))
  (bind _returnvalue_2 (Class|PilgrimQuickplayBPFL|GetUpcomingSong _returnvalue))
  (Utilities|IsValid _returnvalue
    (:Is%20Valid
      (if (Utilities|IsValid _returnvalue_2 :!class /Script/Engine.KismetSystemLibrary)
        (bind _output_get (Variables|Components|SetCurrentSongMetadata _returnvalue_2))
        (CallFunction|DetermineifMissingPart _output_get)))))
