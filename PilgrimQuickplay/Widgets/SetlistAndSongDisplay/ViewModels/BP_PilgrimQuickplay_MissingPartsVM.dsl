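; Quickplay view model for the setlist and song display that reports whether the current song lacks a chart for an instrument part.
; It receives song metadata alone or with a specific part, which sets a flag to use that part.
; Otherwise the check uses the owning player's instrument choice from the Pilgrim player state component.
; A part counts as missing when its track intensity in the metadata equals the missing-intensity value.
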(fn DetermineIfMissingPart
  (Utilities|IsValid (Variables|Components|GetCurrentSongMetadata)
    (:Is%20Valid
      (bind _returnvalue (Class|PilgrimCorePlayerStateComponent|GetPilgrimCorePlayerStateComponent (Variables|/Script/Engine.Controller|GetPlayerState (Sparks|ViewModel|BPGetOwningPlayerController))))
      (Utilities|IsValid _returnvalue
        (:Is%20Valid
          (if (Variables|Default|GetbIsParticularPartInfoNeeded)
            (LocalVariables|SetPartType (Variables|Default|GetCurrent%20Part%20Selected))
            (else
              (LocalVariables|SetPartType (Class|PilgrimCorePlayerStateComponent|GetInstrumentChoice _returnvalue))))
          (return (== (Meta|GetMissingTrackIntensity) (Class|PilgrimSongMetadata|GetTrackIntensity (Variables|Components|GetCurrentSongMetadata) (LocalVariables|GetPartType)))))))))

(fn ReceiveSongMetadata
  (params (Song%20Metadata :type Pilgrim%20Song%20Metadata%20Object%20Reference))
  (Utilities|IsValid Song%20Metadata
    (:Is%20Valid
      (Variables|Components|SetCurrentSongMetadata Song%20Metadata))))

(fn ReceiveSongMetadataAndPart
  (params (Song%20Metadata :type Pilgrim%20Song%20Metadata%20Object%20Reference) (PartSelected :type EPilgrimTrackType%20Enum :default "TrackGuitar"))
  (Utilities|IsValid Song%20Metadata
    (:Is%20Valid
      (Variables|Default|SetbIsParticularPartInfoNeeded true)
      (Variables|Default|SetCurrent%20Part%20Selected PartSelected)
      (Variables|Components|SetCurrentSongMetadata Song%20Metadata))))
